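# Sound cue for the RumbleFemale glider's looping thrust audio, using glider thrust attenuation and the cosmetic glider loop properties sheet.
# A mixer layers two looping waves, a main thrust loop and a sizzle loop.
# An enveloper on the sizzle layer lowers its volume over the first ten seconds.
# Each layer has a continuous modulator that drives pitch and volume from the AnyDirection parameter.
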SoundCue /BRCosmetics/Sounds/Gliders/RumbleFemale/GliderThrustLoop_RumbleFemale1
root=Mixer_0
nodes [
 {
  id=Enveloper_0
  children [WavePlayer_1]
  properties {pitchCurve {editorCurveData {keys [{time=0}]}}, volumeCurve {editorCurveData {keys [<at 0 1>,<at 10 0.625>]}}}
 }
 {id=Mixer_0, children [ModulatorContinuous_1, ModulatorContinuous_0], properties {inputVolume [1,1]}}
 {
  id=ModulatorContinuous_0
  children [WavePlayer_0]
  properties {
   pitchModulationParams {default=0, maxOutput=1.1, minOutput=0.9, parameterName=AnyDirection}
   volumeModulationParams {default=0, maxOutput=1.3, minOutput=1.0369, parameterName=AnyDirection}
  }
 }
 {
  id=ModulatorContinuous_1
  children [Enveloper_0]
  properties {
   pitchModulationParams {default=0, maxOutput=1.1, minOutput=0.9, parameterName=AnyDirection}
   volumeModulationParams {default=0, maxOutput=1.3, minOutput=1.0369, parameterName=AnyDirection}
  }
 }
 {id=WavePlayer_0, properties {bLooping=true, soundWaveAssetPtr=&/Glider_RumbleFemale_ThrustLoop}}
 {
  id=WavePlayer_1
  properties {bLooping=true, soundWaveAssetPtr=&/Glider_RumbleFemale_Sizzle_ThrustLoop}
 }
]
properties {
 attenuationSettings=&Fort_Player_Glider_Thrust_Athena_Att
 audioPropertiesSheet=&APS_Cosmetic_Glider_Loop
 maxDistance=12500
 soundClassObject=&Fort_Gliders_Land
 volumeMultiplier=0.9
 concurrencySet [null]
}
